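# Physics asset for the face accessory part of the male medium Blockstack HopeExtra cosmetic body, covering two collar pieces and both arms.
# Each collar has a capsule on its dyn bone constrained to a tiny capsule on its base bone, with zero twist and swing1 limits and a nonzero swing2 limit for hinge-like motion.
# Unconstrained capsules on the left and right arm_bend_4 bones add arm volumes, likely as colliders for the collars.
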PhysicsAsset /BRCosmetics/Characters/Player/Male/Medium/Bodies/M_MED_Blockstack_HopeExtra/Meshes/Parts/M_MED_Blockstack_HopeExtra_FaceAcc_Physics
bodies {
 Collar_01_base [{shapeType=Capsule, length=0.505, radius=0.505, shapeName=Collar_01_base_capsule, rotation=<rot 0 -90 0>}]
 Collar_01_dyn [{shapeType=Capsule, length=23.4405, radius=6.5, shapeName=Collar_01_dyn_capsule, center=<4.5 -3.8 -2>, rotation=<rot 0 6.9072 -87.0978>}]
 Collar_02_base [{shapeType=Capsule, length=0.505, radius=0.505, shapeName=Collar_02_base_capsule, rotation=<rot 0 -90 0>}]
 Collar_02_dyn [{shapeType=Capsule, length=11.0496, radius=6, shapeName=Collar_02_dyn_capsule, center=<5.5 -2.7 4>, rotation=<rot 0 7.5 -90>}]
 arm_bend_4_l [{shapeType=Capsule, length=27.8232, radius=5, shapeName=arm_bend_3_l_capsule, center=<-2.1685 -1.4286 1.2455>, rotation=<rot 90 0 0>}]
 arm_bend_4_r [{shapeType=Capsule, length=33.4598, radius=5.5, shapeName=arm_bend_3_r_capsule, center=<2.1818 1.032 0.7179>, rotation=<rot 90 177.977 0>}]
}
constraints [
 {bone1Name=Collar_01_dyn, bone2Name=Collar_01_base, swing1LimitDegrees=0, swing2LimitDegrees=55, twistLimitDegrees=0}
 {bone1Name=Collar_02_dyn, bone2Name=Collar_02_base, swing1LimitDegrees=0, swing2LimitDegrees=45, twistLimitDegrees=0}
]
